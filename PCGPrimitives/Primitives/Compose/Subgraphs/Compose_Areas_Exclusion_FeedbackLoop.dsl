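; PCG loop-body subgraph that excludes areas by cutting polygons out of a set carried through a feedback pin.
; Both the Feedback and In inputs are filtered to 2D polygon data only.
; A Polygon Operation subtracts the In polygons, as clip polygons, from the Feedback polygons.
; The difference returns on the Feedback output, so exclusions accumulate over successive iterations.
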(graph Compose_Areas_Exclusion_FeedbackLoop
  (inputs (Feedback :type Any))
  (outputs (Feedback :type Any))
  (bind FilterDataByType_1 (Filter%20Data%20By%20Type :TargetType (struct :Ids [(struct :Struct /Script/PCG.PCGDataTypeInfoPolygon2D)])))
  (bind FilterDataByType_2 (Filter%20Data%20By%20Type :TargetType (struct :Ids [(struct :Struct /Script/PCG.PCGDataTypeInfoPolygon2D)])))
  (bind Polygon2DOperation_0 (Polygon%20Operation :Operation Difference))
  (connect DefaultInputNode Feedback FilterDataByType_1 In)
  (connect DefaultInputNode In FilterDataByType_2 In)
  (connect FilterDataByType_2 InsideFilter Polygon2DOperation_0 ClipPolygons)
  (connect FilterDataByType_1 InsideFilter Polygon2DOperation_0 In)
  (output Feedback Polygon2DOperation_0))
