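; PCG subgraph that serves as the core process for intersecting area primitives as 2D polygons.
; It runs a keep-selected attribute cleanup, then splits the data by the "Input:0" tag into primary and secondary sets, keeping only Polygon2D data in each.
; A Polygon Operation intersects each primary polygon with all secondary ones as clip polygons and outputs the resulting polygons.
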(graph Compose_Areas_Intersection_CoreProcess
  ;[0 0 0 0.5] Core Process
  (bind DeleteAttributes_0 (Delete%20Attributes :Operation KeepSelectedAttributes))
  (bind FilterDataByTag_0 (Filter%20Data%20By%20Tag :SelectedTags "Input:0"))
  (bind FilterDataByType_4 (Filter%20Data%20By%20Type :TargetType (struct :Ids [(struct :Struct /Script/PCG.PCGDataTypeInfoPolygon2D)])))
  (bind FilterDataByType_0 (Filter%20Data%20By%20Type :TargetType (struct :Ids [(struct :Struct /Script/PCG.PCGDataTypeInfoPolygon2D)])))
  (bind Polygon2DOperation_1 (Polygon%20Operation)) ; Intersects each primary input with all secondary input data\nOutputs the resulting polygon data
  (connect DefaultInputNode In DeleteAttributes_0 In)
  (connect DeleteAttributes_0 Out FilterDataByTag_0 In)
  (connect FilterDataByTag_0 InsideFilter FilterDataByType_0 In)
  (connect FilterDataByTag_0 OutsideFilter FilterDataByType_4 In)
  (connect FilterDataByType_4 InsideFilter Polygon2DOperation_1 ClipPolygons)
  (connect FilterDataByType_0 InsideFilter Polygon2DOperation_1 In)
  (output Out Polygon2DOperation_1))
