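; PCG subgraph that serves as the core process of the area union compose primitive, merging input areas into combined polygon data.
; A Delete Attributes node in keep-selected mode first trims input attributes, then a type filter passes only Polygon 2D data.
; A Polygon Operation set to Union merges all the filtered polygons together and outputs the result.
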(graph Compose_Areas_Union_CoreProcess
  ;[0 0 0 0.5] Core Process
  (bind DeleteAttributes_0 (Delete%20Attributes :Operation KeepSelectedAttributes))
  (bind FilterDataByType_0 (Filter%20Data%20By%20Type :TargetType (struct :Ids [(struct :Struct /Script/PCG.PCGDataTypeInfoPolygon2D)])))
  (bind Polygon2DOperation_0 (Polygon%20Operation :Operation Union)) ; Union of all primary input data together\nOutputs the resulting polygon data
  (connect DefaultInputNode In DeleteAttributes_0 In)
  (connect DeleteAttributes_0 Out FilterDataByType_0 In)
  (connect FilterDataByType_0 InsideFilter Polygon2DOperation_0 In)
  (output Out Polygon2DOperation_0))
