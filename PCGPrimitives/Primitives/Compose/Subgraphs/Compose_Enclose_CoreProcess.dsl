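; PCG subgraph that encloses the point, spline and Polygon2D data tagged "Input:0" in one closed boundary spline.
; Points pass through, splines become control points, and polygons become counter-clockwise closed splines before their control points are extracted.
; The merged points feed a 2D convex hull, sorted by descending index into a closed spline forced counter-clockwise.
; The Linear parameter controls whether that spline is linear.
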(graph Compose_Enclose_CoreProcess
  (params (Linear :type Bool))
  ;[0 0 0 0.5 12] Primary/Secondary Inputs
  (bind FilterDataByTag_0 (Filter%20Data%20By%20Tag :SelectedTags "Input:0"))
  ;[0 0 0 0.5] Merge all supported inputs into a single point data
  (bind FilterDataByType_1 (Filter%20Data%20By%20Type :TargetType (struct :Ids [(struct :Struct /Script/PCG.PCGDataTypeInfoPoint)]))) ; Extracts Point data only.
  (bind FilterDataByType_2 (Filter%20Data%20By%20Type :TargetType (struct :Ids [(struct :Struct /Script/PCG.PCGDataTypeInfoSpline)]))) ; Extracts Spline data only.
  (bind FilterDataByType_3 (Filter%20Data%20By%20Type :TargetType (struct :Ids [(struct :Struct /Script/PCG.PCGDataTypeInfoPolygon2D)]))) ; Extracts Polygon2D data only.
  (bind ToPoint_1 (To%20Point))
  (bind CreateSpline_0 (Create%20Spline :bClosedLoop true :InterpTypeAttribute None))
  (bind SplineDirection_0 (Spline%20Direction :Operation ForceCounterClockwise))
  (bind GetSplineControlPoints_0 (Get%20Spline%20Control%20Points))
  (bind MergePoints_1 (Merge%20Points)) ; Merges all point clouds (points + polygon vertices +\nspline control points) into a single\ndataset. The convex hull is computed\nover this combined set.
  ;[0 0 0 0.5] Find convex hull from input(s)
  (bind FindConvexHull2D_1 (Find%20Convex%20Hull%202D)) ; CORE OPERATION: Computes the 2D convex hull\nof all merged points (XY plane).\nOutputs the minimal enclosing polygon that\ncontains every input shape — this IS the\n'enclose' result.
  ;[0 0 0 0.5] Create a closed spline from the convex hull
  (bind SortAttributes_0 (Sort%20Attributes :bUseStableSort false :SortAttributes [(struct :InputSource $Index :SortMethod Descending)]))
  (bind CreateSpline_1 (Create%20Spline :bApplyCustomTangents true :bClosedLoop true :InterpTypeAttribute None :bLinear-param Linear)) ; Builds the enclosing spline from the sorted\nconvex hull vertices.
  (bind SplineDirection_1 (Spline%20Direction :Operation ForceCounterClockwise)) ; Winding correction: forces the\noutput enclosing spline to Counter-Clockwise\norientation. Downstream nodes (e.g. area\nfill, surface sampler) rely on consistent\nCCW winding to treat the spline as a\nclosed region interior.
  (connect ToPoint_1 Out CreateSpline_0 In)
  (connect SortAttributes_0 Out CreateSpline_1 In)
  (connect DefaultInputNode In FilterDataByTag_0 In)
  (connect FilterDataByTag_0 InsideFilter FilterDataByType_1 In)
  (connect FilterDataByTag_0 InsideFilter FilterDataByType_2 In)
  (connect FilterDataByTag_0 InsideFilter FilterDataByType_3 In)
  (connect MergePoints_1 Out FindConvexHull2D_1 In)
  (connect FilterDataByType_2 InsideFilter GetSplineControlPoints_0 In)
  (connect SplineDirection_0 Out GetSplineControlPoints_0 In)
  (connect FilterDataByType_1 InsideFilter MergePoints_1 In)
  (connect GetSplineControlPoints_0 Out MergePoints_1 In)
  (connect FindConvexHull2D_1 Out SortAttributes_0 In)
  (connect CreateSpline_0 Out SplineDirection_0 In)
  (connect CreateSpline_1 Out SplineDirection_1 In)
  (connect FilterDataByType_3 InsideFilter ToPoint_1 In)
  (output Out SplineDirection_1))
