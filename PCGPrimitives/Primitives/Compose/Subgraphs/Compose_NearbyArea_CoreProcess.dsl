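; PCG subgraph that computes the nearby area surrounding input 2D polygons, as the core step of the Compose NearbyArea primitive.
; It keeps only Polygon2D data from the input and expands it by the NearOffset parameter.
; A Difference operation clips the original polygons out of the expanded ones, leaving surrounding bands.
; A final Union merges all bands into the output.
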(graph Compose_NearbyArea_CoreProcess
  (params (NearOffset :type Double :default 1e+02))
  ;[0 0 0 0.5] Core Process
  (bind FilterDataByType_0 (Filter%20Data%20By%20Type :TargetType (struct :Ids [(struct :Struct /Script/PCG.PCGDataTypeInfoPolygon2D)])))
  (bind OffsetPolygon2D_0 (Offset%20Polygon :Offset-param NearOffset)) ; Extends inputs converted to polygon2D(s)
  (bind Polygon2DOperation_0 (Polygon%20Operation :Operation Difference)) ; Subtracts original polygon2D to the extended polygon2D(s)\nand returns the nearby area surrounding the input data.
  (bind Polygon2DOperation_2 (Polygon%20Operation :Operation Union)) ; Union all nearby areas
  (connect DefaultInputNode In FilterDataByType_0 In)
  (connect FilterDataByType_0 InsideFilter OffsetPolygon2D_0 In)
  (connect FilterDataByType_0 InsideFilter Polygon2DOperation_0 ClipPolygons)
  (connect OffsetPolygon2D_0 Out Polygon2DOperation_0 In)
  (connect Polygon2DOperation_0 Out Polygon2DOperation_2 In)
  (output Out Polygon2DOperation_2))
