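; PCG subgraph that serves as the core process of a point difference compose primitive, subtracting secondary point inputs from the primary one.
; The single input is split by the "Input:0" tag into primary and secondary data, and each branch keeps only point data.
; A Difference node removes the secondary points from the primary using binary density in discrete mode, with metadata differencing disabled.
; The result is converted to points for output.
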(graph Compose_Points_Difference_CoreProcess
  (bind ToPoint_0 (To%20Point))
  ;[0 0 0 0.5 12] Primary/Secondary Inputs
  (bind FilterDataByTag_0 (Filter%20Data%20By%20Tag :SelectedTags "Input:0"))
  ;[0 0 0 0.5] Apply binary difference on first input with secondary input point data.
  (bind FilterDataByType_2 (Filter%20Data%20By%20Type :TargetType (struct :Ids [(struct :Struct /Script/PCG.PCGDataTypeInfoPoint)])))
  (bind FilterDataByType_3 (Filter%20Data%20By%20Type :TargetType (struct :Ids [(struct :Struct /Script/PCG.PCGDataTypeInfoPoint)])))
  (bind Difference_0 (Difference :bDiffMetadata false :DensityFunction Binary :Mode Discrete))
  (connect FilterDataByType_3 InsideFilter Difference_0 Differences)
  (connect FilterDataByType_2 InsideFilter Difference_0 Source)
  (connect DefaultInputNode In FilterDataByTag_0 In)
  (connect FilterDataByTag_0 InsideFilter FilterDataByType_2 In)
  (connect FilterDataByTag_0 OutsideFilter FilterDataByType_3 In)
  (connect Difference_0 Out ToPoint_0 In)
  (output Out ToPoint_0))
