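; PCG subgraph that prepares Compose primitive inputs by converting Points and ClosedSplines into Polygon2Ds.
; Points are merged, partitioned per index and looped through convert_point_to_polygon to become square polygons from their XY extents, while closed splines are force-closed into polygons.
; Converted polygons have type tags replaced with Type:Polygon2D, then share the output with untouched Polygon2Ds and Splines.
; Surfaces, Texture2Ds and Volumes inputs are left unconnected.
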(graph Compose_Shared_InputsProcessing
  (inputs (Points :type Point) (Splines :type Spline) (ClosedSplines :type Spline) (Polygon2Ds :type Polygon2D) (Surfaces :type Surface) (Texture2Ds :type BaseTexture) (Volumes :type Volume))
  ;[0 0 0 0.5] Convert Points, Closed Splines to Polygon2Ds and forward Splines
  (bind CreatePolygon2D_0 (Create%20Polygon%202D :InputType ForceClosed)) ; Convert closed splines to polygons
  (bind MergePoints_0 (Merge%20Points :bMergeMetadata false))
  (bind AttributePartition_0 (Attribute%20Partition :PartitionAttributeSelectors [$Index]))
  (bind Loop_0 (Loop :Graph /PCGPrimitives/Primitives/Shared/convert_point_to_polygon.convert_point_to_polygon)) ; Converts each individual input point into a square polygon\nby using the X,Y min and max extents values
  (bind DeleteTags_0 (Delete%20Tags :Operator Substring :SelectedTags "Type:Point, Type:Spline, Type:ClosedSpline, Type:Polygon2D, Type:Surface, Type:Texture2D, Type:Volume")) ; Clean up all incoming type tags
  (bind AddTags_0 (Add%20Tags :TagsToAdd "Type:Polygon2D")) ; Set Type:Polygon2D tag
  (connect DeleteTags_0 Out AddTags_0 In)
  (connect MergePoints_0 Out AttributePartition_0 In)
  (connect DefaultInputNode ClosedSplines CreatePolygon2D_0 In)
  (connect CreatePolygon2D_0 Out DeleteTags_0 In)
  (connect Loop_0 Out DeleteTags_0 In)
  (connect AttributePartition_0 Out Loop_0 In)
  (connect DefaultInputNode Points MergePoints_0 In)
  (output Out AddTags_0)
  (output Out Polygon2Ds)
  (output Out Splines))
