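; PCG subgraph shared by Compose primitives that tags and prepares their data for output.
; It removes empty data, then strips all tags except those in PersistentTags.
; A data count drives a Branch whose B output feeds OutputTaggingSG, alongside a direct link from the tag filter.
; OutputTaggingSG receives DataType and Primitive and supplies the output, while the declared DataSubType parameter is not forwarded.
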(graph Compose_Shared_PrepareOutputs
  (params (DataType :type String :default "None") (DataSubType :type String :default "None") (Primitive :type String :default "None") (PersistentTags :type String))
  ;[0 0 0 0.5] Tag and apply Metadata for Output
  (bind RemoveEmptyData_1 (Remove%20Empty%20Data))
  (bind DeleteTags_0 (Delete%20Tags :Operation KeepOnlySelectedTags :SelectedTags-param PersistentTags))
  (bind DataNum_0 (Data%20Count))
  (bind Branch_0 (Branch))
  (bind Subgraph_0 (Subgraph :Graph /PCGPrimitives/Primitives/Shared/PrepareOutputs/OutputTaggingSG.OutputTaggingSG :DataType DataType :Primitive Primitive))
  (connect DataNum_0 Out Branch_0 bOutputToB)
  (connect DeleteTags_0 Out Branch_0 In)
  (connect DeleteTags_0 Out DataNum_0 In)
  (connect RemoveEmptyData_1 Out DeleteTags_0 In)
  (connect DefaultInputNode In RemoveEmptyData_1 In)
  (connect Branch_0 Output%20B Subgraph_0 In)
  (connect DeleteTags_0 Out Subgraph_0 In)
  (output Out Subgraph_0))
