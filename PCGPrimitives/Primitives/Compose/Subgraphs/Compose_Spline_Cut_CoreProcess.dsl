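; PCG subgraph that performs the core spline cut for a Compose primitive, clipping paths by polygons.
; Input:0 splines, plus polygons rebuilt as closed linear splines with corrected direction, are clipped against Input:1 polygons.
; The Difference_or_intersection switch routes paths to Clip Paths in Difference mode when false or intersection when true.
; A per-polygon difference loop subgraph is also defined, but the main graph does not call it.
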(graph Compose_Spline_Cut_CoreProcess
  (params (Difference_or_intersection :type Bool))
  ;[0 0 0 0.5] Core Process
  (bind DeleteAttributes_0 (Delete%20Attributes :Operation KeepSelectedAttributes))
  (bind FilterDataByTag_0 (Filter%20Data%20By%20Tag :SelectedTags "Input:0"))
  (bind FilterDataByType_1 (Filter%20Data%20By%20Type :TargetType (struct :Ids [(struct :Struct /Script/PCG.PCGDataTypeInfoSpline)])))
  (bind FilterDataByType_0 (Filter%20Data%20By%20Type :TargetType (struct :Ids [(struct :Struct /Script/PCG.PCGDataTypeInfoPolygon2D)])))
  (bind ToPoint_0 (To%20Point))
  (bind CreateSpline_0 (Create%20Spline :bClosedLoop true :bLinear true :InterpTypeAttribute None))
  (bind SplineDirection_0 (Spline%20Direction)) ; because polygon winding is always the same direction
  (bind Branch_0 (Branch :bOutputToB-param Difference_or_intersection))
  (bind FilterDataByTag_1 (Filter%20Data%20By%20Tag :SelectedTags "Input:1"))
  (bind FilterDataByType_3 (Filter%20Data%20By%20Type :TargetType (struct :Ids [(struct :Struct /Script/PCG.PCGDataTypeInfoPolygon2D)])))
  (bind ClipPaths_0 (Clip%20Paths :Operation Difference)) ; Subtract clipping polygons from secondary input
  (bind ClipPaths_1 (Clip%20Paths)) ; Intersect with clipping polygons from secondary input
  (connect FilterDataByType_1 InsideFilter Branch_0 In)
  (connect SplineDirection_0 Out Branch_0 In)
  (connect FilterDataByType_3 InsideFilter ClipPaths_0 ClipPolygons)
  (connect Branch_0 Output%20A ClipPaths_0 In)
  (connect FilterDataByType_3 InsideFilter ClipPaths_1 ClipPolygons)
  (connect Branch_0 Output%20B ClipPaths_1 In)
  (connect ToPoint_0 Out CreateSpline_0 In)
  (connect DefaultInputNode In DeleteAttributes_0 In)
  (connect DeleteAttributes_0 Out FilterDataByTag_0 In)
  (connect DeleteAttributes_0 Out FilterDataByTag_1 In)
  (connect FilterDataByTag_0 InsideFilter FilterDataByType_0 In)
  (connect FilterDataByTag_0 InsideFilter FilterDataByType_1 In)
  (connect FilterDataByTag_1 InsideFilter FilterDataByType_3 In)
  (connect CreateSpline_0 Out SplineDirection_0 In)
  (connect FilterDataByType_0 InsideFilter ToPoint_0 In)
  (output Out ClipPaths_0)
  (output Out ClipPaths_1))

(subgraph loop_for_each_cutting_poly
  (inputs (spline :type Spline))
  (bind ClipPaths_1 (Clip%20Paths :Operation Difference)) ; Subtract clipping polygons from secondary input
  (connect DefaultInputNode In ClipPaths_1 ClipPolygons)
  (connect DefaultInputNode spline ClipPaths_1 In)
  (output Out ClipPaths_1))
